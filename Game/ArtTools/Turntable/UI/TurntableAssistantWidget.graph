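# Editor utility widget with three buttons that help artists fill Fortnite multicapture turntables from level and Content Browser selections.
# The select button collects static meshes from selected actors' components and syncs the Content Browser to their paths, which accumulate because only the mesh set is cleared.
# Add and Replace take the first selected turntable and call its functions for importing selected assets.
# An unconnected loop over stored meshes remains in the replace function.
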EditorUtilityWidgetBlueprint /Game/ArtTools/Turntable/UI/TurntableAssistantWidget
graphs [
 EventGraph {nodes [
   "On Clicked (btnPushMeshes)"!() -> (OutputDelegate, then, widget:User Widget Object Reference, event:Pointer Event Structure) {is=title K2Node_ComponentBoundEvent}
   > pushMeshes()
   "On Clicked (btnStoreMeshes)"!() -> (OutputDelegate, then, widget:User Widget Object Reference, event:Pointer Event Structure) {is=title K2Node_ComponentBoundEvent}
   > Clear!(execute, TargetSet:Set of Static Mesh Object References=Get storedMeshes() -> (storedMeshes:Set of Static Mesh Object References)) -> (then) {is=title K2Node_CallFunction}
   > n1 = GetSelectedLevelActors(self=Get EditorActorSubsystem())
   > n2 = For Each Loop(Array:Array of Actor Object References=@n1) -> (Array Element:Actor Object Reference)
   n3 = To Array!(execute=@n2.Completed, A:Set of Strings=Get storedMeshPaths() -> (storedMeshPaths:Set of Strings)) -> (then, Result:Array of Strings)
   > SyncBrowserToObjects(AssetPaths=@n3)
   n4 = For Each Loop(
    Exec=@n2.LoopBody
    Array:Array of Static Mesh Component Object References=Get Components By Class!(self:Actor Object Reference=@n2, ComponentClass:Actor Component Class Reference=&StaticMeshComponent) -> (ReturnValue:Array of Static Mesh Component Object References)
   ) -> (Array Element:Static Mesh Component Object Reference)
   n5 = Get StaticMesh(self=@n4)
   > Is Valid(InputObject=@n5)
   > Add!(execute, TargetSet:Set of Static Mesh Object References=Get storedMeshes(), NewItem:"Static Mesh Object Reference (by ref)"=@n5) -> (then) {is=title K2Node_CallFunction}
   > Add(TargetSet:Set of Strings=Get storedMeshPaths(), NewItem:"String (by ref)"=Get Object Path String(Object=@n5))
   "On Clicked (btnPushMeshes_Replace)"!() -> (OutputDelegate, then, widget:User Widget Object Reference, event:Pointer Event Structure) {is=title K2Node_ComponentBoundEvent}
   > pushMeshes_Replace()
   n6 = Get storedMeshes()
   PrintString(InString="To String (Integer)"(InInt=Length!(TargetSet:Set of Static Mesh Object References=@n6) -> (ReturnValue:int=0) {is=title K2Node_CallFunction}))
   Event PreConstruct()
   Event Construct()
   Event Tick()
  ]}
 pushMeshes {nodes [
   pushMeshes() {is=title K2Node_FunctionEntry}
   > n1 = Sequence()
   > n2 = GetSelectedLevelActors(self=Get EditorActorSubsystem())
   > n3 = Filter by Class!(
    execute
    TargetArray:Array of Object References=@n2
    ObjectClass:Object Class Reference=/Game/ArtTools/Turntable/Turntables/FortniteMulticaptureTurntable.FortniteMulticaptureTurntable_C
    FilterType:EEditorScriptingFilterType Enum
   ) -> (then, ReturnValue:Array of Fortnite Multicapture Turntable Object References)
   > n4 = Branch(Condition=Is Valid Index!(TargetArray:Array of Fortnite Multicapture Turntable Object References=@n3, IndexToTest:int) -> (ReturnValue:bool))
   > Set l_turntable(l_turntable:Fortnite Multicapture Turntable Object Reference="Get (a copy)"(Array:Array of Fortnite Multicapture Turntable Object References=@n3) -> (Output:Fortnite Multicapture Turntable Object Reference)) -> (Output_Get:Fortnite Multicapture Turntable Object Reference)
   Return Node(execute=@n4.else)
   AddSelectedAssetsToActorClasses(
    execute=@n1.then_1
    self:Fortnite Multicapture Turntable Object Reference=Get l_turntable() -> (l_turntable:Fortnite Multicapture Turntable Object Reference)
   )
  ]}
 pushMeshes_Replace {nodes [
   n1 = To Array(A:Set of Static Mesh Object References=Get storedMeshes()) -> (Result:Array of Static Mesh Object References)
   > n2 = For Each Loop(Array:Array of Object References=@n1) -> (Array Element:Object Reference)
   n3 = Get l_turntable()
   > Add!(
    execute
    TargetArray:Array of Multipass TT Spawn Struct Structures=Get MultiCaptureActorClasses!(self:Multipass Turntable Actor Multi Capture Object Reference=@n3) -> (MultiCaptureActorClasses:Array of Multipass TT Spawn Struct Structures)
    NewItem_OptionalName_2_#1:Name=None
    NewItem_asset_18_#2:Object Reference=@n2
    NewItem_transform_14_#3:Transform="0,0,0|0,0,0|1,1,1"
    NewItem_animation_17_#4:Anim Sequence Base Object Reference
   ) -> (then, ReturnValue) {is=title K2Node_CallArrayFunction}
   pushMeshes_Replace() {is=title K2Node_FunctionEntry}
   > n4 = Sequence()
   > n5 = GetSelectedLevelActors(self=Get EditorActorSubsystem())
   > n6 = Filter by Class(TargetArray=@n5, ObjectClass=&FortniteMulticaptureTurntable_C)
   > n7 = Branch(Condition=Is Valid Index(TargetArray=@n6))
   > Set l_turntable(l_turntable="Get (a copy)"(Array=@n6))
   Return Node(execute=@n7.else)
   ReplaceActorClassesWithSelectedAssets(execute=@n4.then_1, self:Fortnite Multicapture Turntable Object Reference=@n3)
  ]}
]
variables [
 storedMeshes {type="array<StaticMesh>", default []}
 storedMeshPaths {type="array<string>", default []}
]
widgets {
 info {rootWidgetClass=&VerticalBox, widgetCount=4}
 widgets [
  VerticalBox_0 {widgetClassPath=&VerticalBox, widget=VerticalBox_0}
  btnStoreMeshes {
   widgetClassPath=&TA_Generic01_Button
   parent=VerticalBox_0
   widget=btnStoreMeshes
   properties {label=Select Meshes in Content Browser That Are Referenced By Selected Actors}
   slot {class=VerticalBoxSlot}
  }
  btnPushMeshes {
   widgetClassPath=&TA_Generic01_Button
   parent=VerticalBox_0
   widget=btnPushMeshes
   properties {label=Add Selected Assets from Content Browser into Selected Turntable Actors}
   slot {class=VerticalBoxSlot, properties {padding=<ltrb 0 10 0 0>}}
  }
  btnPushMeshes_Replace {
   widgetClassPath=&TA_Generic01_Button
   parent=VerticalBox_0
   widget=btnPushMeshes_Replace
   properties {label=Replace Turntable's Actors with Selected Assets In Content Browser}
   slot {class=VerticalBoxSlot, properties {padding=<ltrb 0 10 0 0>}}
  }
 ]
}
